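# Post-process Anim Blueprint for the large male ChubbyJingle head, adding cheek jiggle and carrying a head and neck rotation-adjustment setup.
# The update graph stores scaled, negated ref-pose rotation deltas for head, neck_02 and neck_01.
# The AnimGraph chains additive bone modifiers for those bones, with no visible link to the stored values, into Anim Dynamics.
# Each cheek stacks separate linear-spring and angular-spring passes, simulated bone-relative with spine_04 as reference.
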AnimBlueprint /BRCosmetics/Characters/Player/Male/Large/Heads/M_LRG_ChubbyJingle_Head/Meshes/M_LRG_ChubbyJingle_Head_PostProcess_AnimBlueprint
target_skeleton=&/M_LRG_ChubbyJingle_Head_Skeleton
graphs [
 AnimGraph {nodes [
   n1 = "Transform (Modify) Bone - Bone: head"!(ComponentPose=Local To Component(LocalPose=Input Pose()), Alpha=1, Rotation="0,0,0") -> (Pose) {
    rotationMode=BMM_Additive
    rotationSpace=BCS_BoneSpace
    translationMode=BMM_Additive
    alphaBoolBlend {blendOption=Linear}
    boneToModify {boneName=head}
   }
   n2 = "Transform (Modify) Bone - Bone: neck_02"!(ComponentPose=@n1, Alpha=1, Rotation="0,0,0") -> (Pose) {
    rotationMode=BMM_Additive
    rotationSpace=BCS_BoneSpace
    translationMode=BMM_Additive
    alphaBoolBlend {blendOption=Linear}
    boneToModify {boneName=neck_02}
   }
   n3 = "Transform (Modify) Bone - Bone: neck_01"!(ComponentPose=@n2, Alpha=1, Rotation="0,0,0") -> (Pose) {
    rotationMode=BMM_Additive
    rotationSpace=BCS_BoneSpace
    translationMode=BMM_Additive
    alphaBoolBlend {blendOption=Linear}
    boneToModify {boneName=neck_01}
   }
   n4 = Anim Dynamics - Bone: L_cheek_wobble(ComponentPose=@n3, Alpha=0.85, ExternalForce="0,0,0") {
    angularDampingOverride=0.999
    angularSpringConstant=20000
    bLinearSpring=true
    bOverrideLinearDamping=true
    bUseGravityOverride=true
    gravityScale=0.35
    linearDampingOverride=0.99
    linearSpringConstant=150
    simulationSpace=BoneRelative
    boundBone {boneName=L_cheek_wobble}
    componentAppliedLinearAccClamp=<0 0 0>
    gravityOverride {z=-980}
    physicsBodyDefinitions [{
      boundBone {boneName=L_cheek_wobble}
      boxExtents=<20 30 30>
      constraintSetup {angularLimitsMax=<0 0 0>, angularLimitsMin=<0 0 0>, angularTarget=<1 0 0>, linearAxesMax=<0.5 0.5 0.5>, linearAxesMin=<-0.5 -0.5 -0.5>}
      localJointOffset=<4 0.5 0>
     }]
    relativeSpaceBone {boneName=spine_04}
    simSpaceSettings {maxAngularAcceleration=10, maxAngularVelocity=300, simSpaceAngularAlpha=1}
   }
   n5 = Anim Dynamics - Bone: L_cheek_wobble(ComponentPose=@n4, Alpha=0.85, ExternalForce="0,0,0") {
    angularDampingOverride=0.999
    angularSpringConstant=20000
    bAngularSpring=true
    bOverrideAngularDamping=true
    bUseGravityOverride=true
    gravityScale=0.35
    simulationSpace=BoneRelative
    boundBone {boneName=L_cheek_wobble}
    componentAppliedLinearAccClamp=<0 0 0>
    gravityOverride {z=-980}
    physicsBodyDefinitions [{
      boundBone {boneName=L_cheek_wobble}
      boxExtents=<20 30 30>
      constraintSetup {angularLimitsMax=<10 10 10>, angularLimitsMin=<-10 -10 -10>, angularTarget=<1 0 0>}
      localJointOffset=<4 0.5 0>
     }]
    relativeSpaceBone {boneName=spine_04}
    simSpaceSettings {maxAngularAcceleration=10, maxAngularVelocity=300, simSpaceAngularAlpha=1}
   }
   n6 = Anim Dynamics - Bone: R_cheek_wobble(ComponentPose=@n5, Alpha=0.85, ExternalForce="0,0,0") {
    angularDampingOverride=0.999
    angularSpringConstant=20000
    bLinearSpring=true
    bOverrideLinearDamping=true
    bUseGravityOverride=true
    gravityScale=0.35
    linearDampingOverride=0.99
    linearSpringConstant=150
    simulationSpace=BoneRelative
    boundBone {boneName=R_cheek_wobble}
    componentAppliedLinearAccClamp=<0 0 0>
    gravityOverride {z=-980}
    physicsBodyDefinitions [{
      boundBone {boneName=R_cheek_wobble}
      boxExtents=<20 30 30>
      constraintSetup {angularLimitsMax=<0 0 0>, angularLimitsMin=<0 0 0>, angularTarget=<1 0 0>, linearAxesMax=<0.5 0.5 0.5>, linearAxesMin=<-0.5 -0.5 -0.5>}
      localJointOffset=<-4 0.5 0>
     }]
    relativeSpaceBone {boneName=spine_04}
    simSpaceSettings {maxAngularAcceleration=10, maxAngularVelocity=300, simSpaceAngularAlpha=1}
   }
   n7 = Anim Dynamics - Bone: R_cheek_wobble(ComponentPose=@n6, Alpha=0.85, ExternalForce="0,0,0") {
    angularDampingOverride=0.999
    angularSpringConstant=20000
    bAngularSpring=true
    bOverrideAngularDamping=true
    bUseGravityOverride=true
    gravityScale=0.35
    simulationSpace=BoneRelative
    boundBone {boneName=R_cheek_wobble}
    componentAppliedLinearAccClamp=<0 0 0>
    gravityOverride {z=-980}
    physicsBodyDefinitions [{
      boundBone {boneName=R_cheek_wobble}
      boxExtents=<20 30 30>
      constraintSetup {angularLimitsMax=<10 10 10>, angularLimitsMin=<-10 -10 -10>, angularTarget=<1 0 0>}
      localJointOffset=<-4 0.5 0>
     }]
    relativeSpaceBone {boneName=spine_04}
    simSpaceSettings {maxAngularAcceleration=10, maxAngularVelocity=300, simSpaceAngularAlpha=1}
   }
   Output Pose(Result=Component To Local(ComponentPose=@n7))
  ]}
 EventGraph {nodes [
   Event BlueprintUpdateAnimation()
   n1 = Get X Factor() -> (X Factor:double)
   > n2 = Rotation adjustment!(
    execute
    self
    X Sign:double=-1
    X factor:double=@n1
    Y Sign:double=-1
    Y factor:double=@n1
    Z Sign:double=-1
    Z factor:double=@n1
    Bone To Limit:Name=head
    Reference Bone:Name=neck_02
   ) -> (then, Limited Rotation:Rotator)
   > Set headRotLimited(headRotLimited:Rotator=@n2) -> (Output_Get:Rotator)
   > n3 = Rotation adjustment(X Sign=-1, X factor=@n1, Y Sign=-1, Y factor=@n1, Z Sign=-1, Z factor=@n1, Bone To Limit=neck_02, Reference Bone=neck_01)
   > Set Neck_02 Rot Limited(Neck_02 Rot Limited:Rotator=@n3) -> (Output_Get:Rotator)
   > n4 = Rotation adjustment(X Sign=-1, X factor=@n1, Y Sign=-1, Y factor=@n1, Z Sign=-1, Z factor=@n1, Bone To Limit=neck_01, Reference Bone=spine_05)
   > Set Neck_01 Rot Limited(Neck_01 Rot Limited:Rotator=@n4) -> (Output_Get:Rotator)
   TryGetPawnOwner()
  ]}
 Rotation adjustment {nodes [
   n1 = Rotation adjustment!() -> (
    then
    X Sign:double=~
    X factor:double=~
    Y Sign:double=~
    Y factor:double=~
    Z Sign:double=~
    Z factor:double=~
    Bone To Limit:Name
    Reference Bone:Name
   ) {is=title K2Node_FunctionEntry}
   n2 = GetOwnerSkeletalMeshComponent()
   n3 = BreakRotator(InRot=BreakTransform(InTransform=GetDeltaTransformFromRefPose(self=@n2, BoneName=@n1.Bone To Limit, BaseName=@n1.Reference Bone)).Rotation)
   > Return Node!(
    execute
    Limited Rotation:Rotator=MakeRotator(
     Roll=float * float!(A=@n3, B=@n1, C:double=@n1.X factor) -> (ReturnValue)
     Pitch=float * float(A=@n3.Pitch, B=@n1.Y Sign, C=@n1.Y factor)
     Yaw=float * float(A=@n3.Yaw, B=@n1.Z Sign, C=@n1.Z factor)
    )
   )
  ]}
]
variables [
 headRotLimited {type=Rotator}
 Neck_01 Rot Limited {type=Rotator}
 Neck_02 Rot Limited {type=Rotator}
 X Factor {type=number, default=0.8}
]
